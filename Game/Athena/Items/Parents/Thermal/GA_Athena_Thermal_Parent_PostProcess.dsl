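; Parent gameplay ability for thermal items that shows a thermal post process while active and can report a quest stat.
; On activation it adds the PP_Cue gameplay cue to the owner, and removes it when the ability ends.
; It also sends TC_QuestTargetTags to the player's quest manager when they equal NullTagContainer, although the graph comment says the event is for valid tags.
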(event Ability|EventActivateAbility
  ;[1 1 1] Add Thermal Post Process and send stat event if there are valid tags
  (bind _asfort_player_controller (Utilities|Casting|CastToFortPlayerController (Pawn|GetController (Ability|GetActivatingPawn))))
  (if (GameplayTags|Equal(GameplayTagContainer) (Variables|ThermalParent|GetTC_QuestTargetTags) (Variables|ThermalParent|GetNullTagContainer))
    (Quest|SendComplexCustomStatEvent (Quests|GetQuestManager _asfort_player_controller "Invalid") 0 "(GameplayTags=)" (Variables|ThermalParent|GetTC_QuestTargetTags) 1)
    (else
      _))
  (Ability|AddGameplayCueWithParamsToOwner (Variables|ThermalParent|GetPP_Cue) (Ability|GameplayCue|MakeGameplayCueParameters 0.0) false))

(event Ability|EventOnEndAbility
  (params (bWasCancelled :type Boolean))
  ;[1 1 1] Remove Post
  (Ability|RemoveGameplayCueFromOwner (Variables|ThermalParent|GetPP_Cue)))
